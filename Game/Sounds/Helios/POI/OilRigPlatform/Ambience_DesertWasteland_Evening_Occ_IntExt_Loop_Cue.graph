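# Looping desert wasteland evening wind ambience cue for the Oil Rig Platform POI, played quietly and pitched down with Athena ambient attenuation and sound class.
# A quality-level root routes its first two slots to the quad wind loop and its last two to the stereo version.
# Each loop passes through two endlessly repeating envelopers of different lengths, one varying pitch and the other gently dipping volume, adding slow drift.
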SoundCue /Game/Sounds/Helios/POI/OilRigPlatform/Ambience_DesertWasteland_Evening_Occ_IntExt_Loop_Cue
root=QualityLevel_0
nodes [
 {
  id=Enveloper_0
  children [WavePlayer_2]
  properties {
   bLoop=true
   bLoopIndefinitely=true
   loopEnd=30
   pitchCurve {editorCurveData {keys [
      {arriveTangent=-0.0013, interpMode=RCIM_Cubic, leaveTangent=-0.0013, time=0}
      <at 15.9468 0.98 RCIM_Cubic>
      {arriveTangent=0.0014, interpMode=RCIM_Cubic, leaveTangent=0.0014, time=30}
     ]}}
   volumeCurve {editorCurveData {keys [<at 0 1>]}}
  }
 }
 {
  id=Enveloper_1
  children [WavePlayer_0]
  properties {
   bLoop=true
   bLoopIndefinitely=true
   loopEnd=30
   pitchCurve {editorCurveData {keys [
      {arriveTangent=-0.0013, interpMode=RCIM_Cubic, leaveTangent=-0.0013, time=0}
      <at 15.9468 0.98 RCIM_Cubic>
      {arriveTangent=0.0014, interpMode=RCIM_Cubic, leaveTangent=0.0014, time=30}
     ]}}
   volumeCurve {editorCurveData {keys [<at 0 1>]}}
  }
 }
 {
  id=Enveloper_2
  children [Enveloper_1]
  properties {
   bLoop=true
   bLoopIndefinitely=true
   loopEnd=20
   pitchCurve {editorCurveData {keys [{time=0}]}}
   volumeCurve {editorCurveData {keys [
      {arriveTangent=-0.01, interpMode=RCIM_Cubic, leaveTangent=-0.01, time=0, value=1}
      <at 10 0.9 RCIM_Cubic>
      {arriveTangent=0.01, interpMode=RCIM_Cubic, leaveTangent=0.01, time=20, value=1}
     ]}}
  }
 }
 {
  id=Enveloper_3
  children [Enveloper_0]
  properties {
   bLoop=true
   bLoopIndefinitely=true
   loopEnd=20
   pitchCurve {editorCurveData {keys [{time=0}]}}
   volumeCurve {editorCurveData {keys [
      {arriveTangent=-0.01, interpMode=RCIM_Cubic, leaveTangent=-0.01, time=0, value=1}
      <at 10 0.9 RCIM_Cubic>
      {arriveTangent=0.01, interpMode=RCIM_Cubic, leaveTangent=0.01, time=20, value=1}
     ]}}
  }
 }
 {id=QualityLevel_0, children [Enveloper_2,Enveloper_2,Enveloper_3,Enveloper_3]}
 {
  id=WavePlayer_0
  properties {bLooping=true, soundWaveAssetPtr=/Game/Sounds/Helios/Biomes/DesertWasteland/Ambience_Desert_Evening_Wind_Loop_01_Quad}
 }
 {
  id=WavePlayer_2
  properties {bLooping=true, soundWaveAssetPtr=/Game/Sounds/Helios/Biomes/DesertWasteland/Ambience_Desert_Evening_Wind_Loop_01_Stereo}
 }
]
properties {
 attenuationSettings=&Fort_Ambient_Athena_Att
 maxDistance=2097152
 pitchMultiplier=0.85
 soundClassObject=&Fort_Ambient_Athena
 volumeMultiplier=0.15
 preEffectBusSends [{}]
}
