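# Looping ambient sound cue for the windy gantry at the Helios oil rig platform POI.
# It mixes a reused grasslands night bed with one of two randomly chosen, pitch-modulated distant metal rattle loops.
# Nested endlessly looping envelopers of differing lengths drive pitch and volume, giving the rattle uneven swells and dips over the steadier bed.
# The quality node feeds this mix to its first two slots and leaves the other two empty.
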SoundCue /Game/Sounds/Helios/POI/OilRigPlatform/Ambience_OilRig_Windy_Gantry_Loop_Cue
root=QualityLevel_0
nodes [
 {
  id=Enveloper_0
  children [Enveloper_3]
  properties {
   bLoop=true
   bLoopIndefinitely=true
   loopEnd=100
   pitchCurve {editorCurveData {keys [<at 0 0.8>]}}
   volumeCurve {editorCurveData {keys [<at 0 1 RCIM_Cubic>,<at 50 0.9 RCIM_Cubic>,<at 100 1 RCIM_Cubic>]}}
  }
 }
 {
  id=Enveloper_2
  children [Modulator_0]
  properties {
   bLoop=true
   bLoopIndefinitely=true
   loopEnd=30
   pitchCurve {editorCurveData {keys [{interpMode=RCIM_Cubic, time=0}, <at 15.9468 0.9181 RCIM_Cubic>, {interpMode=RCIM_Cubic, time=30}]}}
   volumeCurve {editorCurveData {keys [<at 0 1>]}}
  }
 }
 {
  id=Enveloper_3
  children [Mixer_3]
  properties {
   bLoop=true
   bLoopIndefinitely=true
   loopEnd=30
   pitchCurve {editorCurveData {keys [{interpMode=RCIM_Cubic, time=0}, <at 15.9468 0.9181 RCIM_Cubic>, {interpMode=RCIM_Cubic, time=30}]}}
   volumeCurve {editorCurveData {keys [<at 0 1>]}}
  }
 }
 {
  id=Enveloper_4
  children [Enveloper_2]
  properties {
   bLoop=true
   bLoopIndefinitely=true
   loopEnd=40
   pitchCurve {editorCurveData {keys [{time=0}]}}
   volumeCurve {editorCurveData {keys [
      {arriveTangent=-0.01, interpMode=RCIM_Cubic, time=0, value=1}
      <at 10.0039 0.2703 RCIM_Cubic>
      {arriveTangent=-0.0105, interpMode=RCIM_Cubic, leaveTangent=-0.0105, tangentMode=RCTM_User, time=20, value=1}
      {arriveTangent=-0.025, interpMode=RCIM_Cubic, leaveTangent=-0.025, time=23.5545, value=0.0703}
      {arriveTangent=0.0013, interpMode=RCIM_Cubic, leaveTangent=0.0013, tangentMode=RCTM_User, time=28, value=0.8}
      {arriveTangent=0.0167, interpMode=RCIM_Cubic, leaveTangent=0.0167, time=33.0039, value=0.2703}
      <at 40 1>
     ]}}
  }
 }
 {id=Mixer_2, children [Enveloper_0,Enveloper_4], properties {inputVolume [1,1]}}
 {id=Mixer_3, children [WavePlayer_1], properties {inputVolume [2]}}
 {id=Modulator_0, children [Random_1], properties {pitchMax=0.9, pitchMin=0.8}}
 {id=QualityLevel_0, children [Mixer_2,Mixer_2,null,null]}
 {id=Random_1, children [WavePlayer_3,WavePlayer_6], properties {weights [1,1]}}
 {
  id=WavePlayer_1
  properties {bLooping=true, soundWaveAssetPtr=/Game/Sounds/Artemis/Biomes/Grasslands/Ambience_Grasslands_Night_No_Crickets_Quad}
 }
 {
  id=WavePlayer_3
  properties {bLooping=true, soundWaveAssetPtr=/Game/Sounds/Asteria/POI/JunkYard/Ambience_Metal_Distant_Rattling_In_Wind_Loop_03_Quad}
 }
 {
  id=WavePlayer_6
  properties {bLooping=true, soundWaveAssetPtr=/Game/Sounds/Asteria/POI/JunkYard/Ambience_Metal_Distant_Rattling_In_Wind_Loop_04_Quad}
 }
]
properties {maxDistance=2097152, soundClassObject=&Fort_Ambient_Athena, volumeMultiplier=0.251}
